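# Master Level Sequence for the four-player Quickplay band intro teleport animatic, combining a cinematic shot, a subsequence and spawn-tracked VFX actors.
# Two rift portal Niagara effects hand off early at one shared point.
# Later, a poof, four player glows with individual EndPosition offsets and four lightning bolts fire almost together.
# Four blasts follow as paired hex and cylinder floor pad glows scale up and vanish at four landing spots.
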LevelSequence /PilgrimQuickplay/BandIntro/Teleport/BandIntro_Animatic_MASTER/LS_BI_AM_MASTER_4P
bindings [
 NiagaraComponent0 {
  id=#1
  parent_id=#2
  possessed_class=NiagaraComponent
  tracks [
   {
    type=MovieSceneParticleTrack
    display_name=FX System
    sections [{type=MovieSceneParticleSection, channels [None {type=MovieSceneScriptingParticleChannel}]}]
   }
   {type=MovieSceneNiagaraSystemTrack, sections [{type=MovieSceneNiagaraSystemSpawnSection, end_frame=469, start_frame=169}]}
  ]
 }
 SM Hex Floor Pad Glow {
  id=#3
  possessed_class=FortStaticMeshActor
  tracks [
   {
    type=MovieScene3DTransformTrack
    sections [{
      type=MovieScene3DTransformSection
      channels [
       Location.X {type=MovieSceneScriptingDoubleChannel, default=0, const [-7153.3203,1]}
       Location.Y {type=MovieSceneScriptingDoubleChannel, default=0, const [-901.6919,1]}
       Location.Z {type=MovieSceneScriptingDoubleChannel, default=0, const [536.5154,1]}
       Rotation.X {type=MovieSceneScriptingDoubleChannel, default=0, const [0,3]}
       Rotation.Y {type=MovieSceneScriptingDoubleChannel, default=0, const [0,3]}
       Rotation.Z {type=MovieSceneScriptingDoubleChannel, default=0, const [0,3]}
       Scale.X {type=MovieSceneScriptingDoubleChannel, default=1, keys [[175,0.001],[183.0013,0.001],[184.0013,0.1],[196.0013,1],[197.0013,0.0001]]}
       Scale.Y {type=MovieSceneScriptingDoubleChannel, default=1, keys [[175,0.001],[183.0013,0.001],[184.0013,0.1],[196.0013,1],[197.0013,0.0001]]}
       Scale.Z {
        type=MovieSceneScriptingDoubleChannel
        default=1
        keys [[175,0.001],[183.0013,0.001],[184.0013,0.1],[188.79,1.5],[196.0013,0.05],[197.0013,0]]
       }
      ]
     }]
   }
   {
    type=MovieSceneSpawnTrack
    sections [{type=MovieSceneSpawnSection, channels [None {type=MovieSceneScriptingBoolChannel, default=true, const [true,1]}]}]
   }
  ]
 }
 SM Hex Floor Pad Glow {
  id=#4
  possessed_class=FortStaticMeshActor
  tracks [
   {
    type=MovieScene3DTransformTrack
    sections [{
      type=MovieScene3DTransformSection
      channels [
       Location.X {type=MovieSceneScriptingDoubleChannel, default=0, const [-7447.4619,1]}
       Location.Y {type=MovieSceneScriptingDoubleChannel, default=0, const [-716.4758,1]}
       Location.Z {type=MovieSceneScriptingDoubleChannel, default=0, const [653.6868,1]}
       Rotation.X {type=MovieSceneScriptingDoubleChannel, default=0, const [0,3]}
       Rotation.Y {type=MovieSceneScriptingDoubleChannel, default=0, const [0,3]}
       Rotation.Z {type=MovieSceneScriptingDoubleChannel, default=0, const [0,3]}
       Scale.X {type=MovieSceneScriptingDoubleChannel, default=1, keys [[175,0.001],[183.0013,0.001],[184.0013,0.1],[196.0013,1],[197.0013,0.0001]]}
       Scale.Y {type=MovieSceneScriptingDoubleChannel, default=1, keys [[175,0.001],[183.0013,0.001],[184.0013,0.1],[196.0013,1],[197.0013,0.0001]]}
       Scale.Z {
        type=MovieSceneScriptingDoubleChannel
        default=1
        keys [[175,0.001],[183.0013,0.001],[184.0013,0.1],[188.79,1.5],[196.0013,0.05],[197.0013,0]]
       }
      ]
     }]
   }
   {
    type=MovieSceneSpawnTrack
    sections [{type=MovieSceneSpawnSection, channels [None {type=MovieSceneScriptingBoolChannel, default=true, const [true,1]}]}]
   }
  ]
 }
 NS_Sparks_Lightning_Teleport_Blast {
  id=#5
  possessed_class=NiagaraActor
  tracks [
   {type=MovieSceneSpawnTrack, sections [{type=MovieSceneSpawnSection, channels [None {type=MovieSceneScriptingBoolChannel, default=true}]}]}
   {
    type=MovieScene3DTransformTrack
    sections [{
      type=MovieScene3DTransformSection
      channels [
       Location.X {type=MovieSceneScriptingDoubleChannel, default=-7441.5343}
       Location.Y {type=MovieSceneScriptingDoubleChannel, default=-608.9357}
       Location.Z {type=MovieSceneScriptingDoubleChannel, default=785.6868}
       Rotation.X {type=MovieSceneScriptingDoubleChannel, default=0}
       Rotation.Y {type=MovieSceneScriptingDoubleChannel, default=0}
       Rotation.Z {type=MovieSceneScriptingDoubleChannel, default=65.0436}
       Scale.X {type=MovieSceneScriptingDoubleChannel, default=1}
       Scale.Y {type=MovieSceneScriptingDoubleChannel, default=1}
       Scale.Z {type=MovieSceneScriptingDoubleChannel, default=1}
      ]
     }]
   }
  ]
 }
 NiagaraComponent0 {
  id=#6
  parent_id=#7
  possessed_class=NiagaraComponent
  tracks [
   {
    type=MovieSceneParticleTrack
    display_name=FX System
    sections [{type=MovieSceneParticleSection, channels [None {type=MovieSceneScriptingParticleChannel}]}]
   }
   {type=MovieSceneNiagaraSystemTrack, sections [{type=MovieSceneNiagaraSystemSpawnSection, end_frame=466, start_frame=166}]}
   {
    type=MovieSceneNiagaraVectorParameterTrack
    display_name=EndPosition
    sections [{
      type=MovieSceneFloatVectorSection
      channels [
       Vector.X {type=MovieSceneScriptingFloatChannel, default=-50}
       Vector.Y {type=MovieSceneScriptingFloatChannel, default=150}
       Vector.Z {type=MovieSceneScriptingFloatChannel, default=-280}
      ]
     }]
   }
  ]
 }
 NiagaraComponent0 {
  id=#8
  parent_id=#9
  possessed_class=NiagaraComponent
  tracks [
   {
    type=MovieSceneParticleTrack
    display_name=FX System
    sections [{type=MovieSceneParticleSection, channels [None {type=MovieSceneScriptingParticleChannel}]}]
   }
   {type=MovieSceneNiagaraSystemTrack, sections [{type=MovieSceneNiagaraSystemSpawnSection, end_frame=469, start_frame=169}]}
  ]
 }
 NS_Sparks_Lightning_Teleport_Bolt {
  id=#2
  possessed_class=NiagaraActor
  tracks [
   {type=MovieSceneSpawnTrack, sections [{type=MovieSceneSpawnSection, channels [None {type=MovieSceneScriptingBoolChannel, default=true}]}]}
   {
    type=MovieScene3DTransformTrack
    sections [{
      type=MovieScene3DTransformSection
      channels [
       Location.X {type=MovieSceneScriptingDoubleChannel, default=-6890.322}
       Location.Y {type=MovieSceneScriptingDoubleChannel, default=-470.7068}
       Location.Z {type=MovieSceneScriptingDoubleChannel, default=630.6797}
       Rotation.X {type=MovieSceneScriptingDoubleChannel, default=-21.1302}
       Rotation.Y {type=MovieSceneScriptingDoubleChannel, default=-5.5519}
       Rotation.Z {type=MovieSceneScriptingDoubleChannel, default=8.2506}
       Scale.X {type=MovieSceneScriptingDoubleChannel, default=1}
       Scale.Y {type=MovieSceneScriptingDoubleChannel, default=1}
       Scale.Z {type=MovieSceneScriptingDoubleChannel, default=0.65}
      ]
     }]
   }
  ]
 }
 NS Rift Teleport Portal_second {
  id=#10
  possessed_class=NiagaraActor
  tracks [
   {
    type=MovieSceneSpawnTrack
    sections [{type=MovieSceneSpawnSection, channels [None {type=MovieSceneScriptingBoolChannel, default=false, keys [[52,false],[80,true]]}]}]
   }
   {
    type=MovieScene3DTransformTrack
    sections [{
      type=MovieScene3DTransformSection
      channels [
       Location.X {type=MovieSceneScriptingDoubleChannel, default=-6849.1834}
       Location.Y {type=MovieSceneScriptingDoubleChannel, default=-603.2899}
       Location.Z {type=MovieSceneScriptingDoubleChannel, default=892.6262}
       Rotation.X {type=MovieSceneScriptingDoubleChannel, default=0}
       Rotation.Y {type=MovieSceneScriptingDoubleChannel, default=0}
       Rotation.Z {type=MovieSceneScriptingDoubleChannel, default=0}
       Scale.X {type=MovieSceneScriptingDoubleChannel, default=1}
       Scale.Y {type=MovieSceneScriptingDoubleChannel, default=1}
       Scale.Z {type=MovieSceneScriptingDoubleChannel, default=1}
      ]
     }]
   }
  ]
 }
 NiagaraComponent0 {
  id=#11
  parent_id=#12
  possessed_class=NiagaraComponent
  tracks [
   {
    type=MovieSceneParticleTrack
    display_name=FX System
    sections [{type=MovieSceneParticleSection, channels [None {type=MovieSceneScriptingParticleChannel}]}]
   }
   {type=MovieSceneNiagaraSystemTrack, sections [{type=MovieSceneNiagaraSystemSpawnSection, end_frame=466, start_frame=166}]}
   {
    type=MovieSceneNiagaraVectorParameterTrack
    display_name=EndPosition
    sections [{
      type=MovieSceneFloatVectorSection
      channels [
       Vector.X {type=MovieSceneScriptingFloatChannel, default=-250}
       Vector.Y {type=MovieSceneScriptingFloatChannel, default=-210}
       Vector.Z {type=MovieSceneScriptingFloatChannel, default=-180}
      ]
     }]
   }
  ]
 }
 NS_Rift_Teleport_PlayerGlow {
  id=#12
  possessed_class=NiagaraActor
  tracks [
   {type=MovieSceneSpawnTrack, sections [{type=MovieSceneSpawnSection, channels [None {type=MovieSceneScriptingBoolChannel, default=true}]}]}
   {
    type=MovieScene3DTransformTrack
    sections [{
      type=MovieScene3DTransformSection
      channels [
       Location.X {type=MovieSceneScriptingDoubleChannel, default=-6802.2574}
       Location.Y {type=MovieSceneScriptingDoubleChannel, default=-588.7739}
       Location.Z {type=MovieSceneScriptingDoubleChannel, default=892.6262}
       Rotation.X {type=MovieSceneScriptingDoubleChannel, default=0}
       Rotation.Y {type=MovieSceneScriptingDoubleChannel, default=0}
       Rotation.Z {type=MovieSceneScriptingDoubleChannel, default=0}
       Scale.X {type=MovieSceneScriptingDoubleChannel, default=1}
       Scale.Y {type=MovieSceneScriptingDoubleChannel, default=1}
       Scale.Z {type=MovieSceneScriptingDoubleChannel, default=1}
      ]
     }]
   }
  ]
 }
 NiagaraComponent0 {
  id=#13
  parent_id=#14
  possessed_class=NiagaraComponent
  tracks [
   {
    type=MovieSceneParticleTrack
    display_name=FX System
    sections [{type=MovieSceneParticleSection, channels [None {type=MovieSceneScriptingParticleChannel}]}]
   }
   {type=MovieSceneNiagaraSystemTrack, sections [{type=MovieSceneNiagaraSystemSpawnSection, end_frame=480, start_frame=180}]}
  ]
 }
 NS_Sparks_Lightning_Teleport_Bolt {
  id=#15
  possessed_class=NiagaraActor
  tracks [
   {type=MovieSceneSpawnTrack, sections [{type=MovieSceneSpawnSection, channels [None {type=MovieSceneScriptingBoolChannel, default=true}]}]}
   {
    type=MovieScene3DTransformTrack
    sections [{
      type=MovieScene3DTransformSection
      channels [
       Location.X {type=MovieSceneScriptingDoubleChannel, default=-7229.3271}
       Location.Y {type=MovieSceneScriptingDoubleChannel, default=-222.0148}
       Location.Z {type=MovieSceneScriptingDoubleChannel, default=614.8303}
       Rotation.X {type=MovieSceneScriptingDoubleChannel, default=-62.9532}
       Rotation.Y {type=MovieSceneScriptingDoubleChannel, default=-2.9057}
       Rotation.Z {type=MovieSceneScriptingDoubleChannel, default=47.5552}
       Scale.X {type=MovieSceneScriptingDoubleChannel, default=1}
       Scale.Y {type=MovieSceneScriptingDoubleChannel, default=1}
       Scale.Z {type=MovieSceneScriptingDoubleChannel, default=1.2}
      ]
     }]
   }
  ]
 }
 NiagaraComponent0 {
  id=#16
  parent_id=#17
  possessed_class=NiagaraComponent
  tracks [
   {
    type=MovieSceneParticleTrack
    display_name=FX System
    sections [{type=MovieSceneParticleSection, channels [None {type=MovieSceneScriptingParticleChannel}]}]
   }
   {type=MovieSceneNiagaraSystemTrack, sections [{type=MovieSceneNiagaraSystemSpawnSection, end_frame=469, start_frame=169}]}
  ]
 }
 NS_Rift_Teleport_PlayerGlow {
  id=#7
  possessed_class=NiagaraActor
  tracks [
   {type=MovieSceneSpawnTrack, sections [{type=MovieSceneSpawnSection, channels [None {type=MovieSceneScriptingBoolChannel, default=true}]}]}
   {
    type=MovieScene3DTransformTrack
    sections [{
      type=MovieScene3DTransformSection
      channels [
       Location.X {type=MovieSceneScriptingDoubleChannel, default=-6849.1836}
       Location.Y {type=MovieSceneScriptingDoubleChannel, default=-603.2899}
       Location.Z {type=MovieSceneScriptingDoubleChannel, default=892.6262}
       Rotation.X {type=MovieSceneScriptingDoubleChannel, default=0}
       Rotation.Y {type=MovieSceneScriptingDoubleChannel, default=0}
       Rotation.Z {type=MovieSceneScriptingDoubleChannel, default=0}
       Scale.X {type=MovieSceneScriptingDoubleChannel, default=1}
       Scale.Y {type=MovieSceneScriptingDoubleChannel, default=1}
       Scale.Z {type=MovieSceneScriptingDoubleChannel, default=1}
      ]
     }]
   }
  ]
 }
 NS_Sparks_Lightning_Teleport_Blast {
  id=#18
  possessed_class=NiagaraActor
  tracks [
   {type=MovieSceneSpawnTrack, sections [{type=MovieSceneSpawnSection, channels [None {type=MovieSceneScriptingBoolChannel, default=true}]}]}
   {
    type=MovieScene3DTransformTrack
    sections [{
      type=MovieScene3DTransformSection
      channels [
       Location.X {type=MovieSceneScriptingDoubleChannel, default=-7212.9308}
       Location.Y {type=MovieSceneScriptingDoubleChannel, default=-223.3734}
       Location.Z {type=MovieSceneScriptingDoubleChannel, default=633.0222}
       Rotation.X {type=MovieSceneScriptingDoubleChannel, default=0}
       Rotation.Y {type=MovieSceneScriptingDoubleChannel, default=0}
       Rotation.Z {type=MovieSceneScriptingDoubleChannel, default=102.4108}
       Scale.X {type=MovieSceneScriptingDoubleChannel, default=1}
       Scale.Y {type=MovieSceneScriptingDoubleChannel, default=1}
       Scale.Z {type=MovieSceneScriptingDoubleChannel, default=1}
      ]
     }]
   }
  ]
 }
 NS_Rift_Teleport_PlayerGlow {
  id=#19
  possessed_class=NiagaraActor
  tracks [
   {type=MovieSceneSpawnTrack, sections [{type=MovieSceneSpawnSection, channels [None {type=MovieSceneScriptingBoolChannel, default=true}]}]}
   {
    type=MovieScene3DTransformTrack
    sections [{
      type=MovieScene3DTransformSection
      channels [
       Location.X {type=MovieSceneScriptingDoubleChannel, default=-6849.1836}
       Location.Y {type=MovieSceneScriptingDoubleChannel, default=-603.2899}
       Location.Z {type=MovieSceneScriptingDoubleChannel, default=892.6262}
       Rotation.X {type=MovieSceneScriptingDoubleChannel, default=0}
       Rotation.Y {type=MovieSceneScriptingDoubleChannel, default=0}
       Rotation.Z {type=MovieSceneScriptingDoubleChannel, default=0}
       Scale.X {type=MovieSceneScriptingDoubleChannel, default=1}
       Scale.Y {type=MovieSceneScriptingDoubleChannel, default=1}
       Scale.Z {type=MovieSceneScriptingDoubleChannel, default=1}
      ]
     }]
   }
  ]
 }
 SM Cylinder Floor Pad Glow {
  id=#20
  possessed_class=FortStaticMeshActor
  tracks [
   {
    type=MovieScene3DTransformTrack
    sections [{
      type=MovieScene3DTransformSection
      channels [
       Location.X {type=MovieSceneScriptingDoubleChannel, default=0, const [-6904.856,1]}
       Location.Y {type=MovieSceneScriptingDoubleChannel, default=0, const [-587.4029,1]}
       Location.Z {type=MovieSceneScriptingDoubleChannel, default=0, const [497.0325,1]}
       Rotation.X {type=MovieSceneScriptingDoubleChannel, default=0, const [0,3]}
       Rotation.Y {type=MovieSceneScriptingDoubleChannel, default=0, const [0,3]}
       Rotation.Z {type=MovieSceneScriptingDoubleChannel, default=0, const [0,3]}
       Scale.X {type=MovieSceneScriptingDoubleChannel, default=1, keys [[175,0.001],[183.0013,0.001],[184.0013,0.1],[196.0013,1],[197.0013,0.0001]]}
       Scale.Y {type=MovieSceneScriptingDoubleChannel, default=1, keys [[175,0.001],[183.0013,0.001],[184.0013,0.1],[196.0013,1],[197.0013,0.0001]]}
       Scale.Z {
        type=MovieSceneScriptingDoubleChannel
        default=1
        keys [[175,0.001],[183.0013,0.001],[184.0013,0.1],[188.79,1.5],[196.0013,0.05],[197.0013,0]]
       }
      ]
     }]
   }
   {
    type=MovieSceneSpawnTrack
    sections [{type=MovieSceneSpawnSection, channels [None {type=MovieSceneScriptingBoolChannel, default=true, const [true,1]}]}]
   }
  ]
 }
 NiagaraComponent0 {
  id=#21
  parent_id=#22
  possessed_class=NiagaraComponent
  tracks [
   {
    type=MovieSceneParticleTrack
    display_name=FX System
    sections [{type=MovieSceneParticleSection, channels [None {type=MovieSceneScriptingParticleChannel}]}]
   }
   {type=MovieSceneNiagaraSystemTrack, sections [{type=MovieSceneNiagaraSystemSpawnSection, end_frame=466, start_frame=166}]}
   {
    type=MovieSceneNiagaraVectorParameterTrack
    display_name=EndPosition
    sections [{
      type=MovieSceneFloatVectorSection
      channels [
       Vector.X {type=MovieSceneScriptingFloatChannel, default=-400}
       Vector.Y {type=MovieSceneScriptingFloatChannel, default=400}
       Vector.Z {type=MovieSceneScriptingFloatChannel, default=-280}
      ]
     }]
   }
  ]
 }
 SM Hex Floor Pad Glow {
  id=#23
  possessed_class=FortStaticMeshActor
  tracks [
   {
    type=MovieScene3DTransformTrack
    sections [{
      type=MovieScene3DTransformSection
      channels [
       Location.X {type=MovieSceneScriptingDoubleChannel, default=0, const [-7207.5327,1]}
       Location.Y {type=MovieSceneScriptingDoubleChannel, default=0, const [-340.961,1]}
       Location.Z {type=MovieSceneScriptingDoubleChannel, default=0, const [501.0222,1]}
       Rotation.X {type=MovieSceneScriptingDoubleChannel, default=0, const [0,3]}
       Rotation.Y {type=MovieSceneScriptingDoubleChannel, default=0, const [0,3]}
       Rotation.Z {type=MovieSceneScriptingDoubleChannel, default=0, const [0,3]}
       Scale.X {type=MovieSceneScriptingDoubleChannel, default=1, keys [[175,0.001],[183.0013,0.001],[184.0013,0.1],[196.0013,1],[197.0013,0.0001]]}
       Scale.Y {type=MovieSceneScriptingDoubleChannel, default=1, keys [[175,0.001],[183.0013,0.001],[184.0013,0.1],[196.0013,1],[197.0013,0.0001]]}
       Scale.Z {
        type=MovieSceneScriptingDoubleChannel
        default=1
        keys [[175,0.001],[183.0013,0.001],[184.0013,0.1],[188.79,1.5],[196.0013,0.05],[197.0013,0]]
       }
      ]
     }]
   }
   {
    type=MovieSceneSpawnTrack
    sections [{type=MovieSceneSpawnSection, channels [None {type=MovieSceneScriptingBoolChannel, default=true, const [true,1]}]}]
   }
  ]
 }
 SM Cylinder Floor Pad Glow {
  id=#24
  possessed_class=FortStaticMeshActor
  tracks [
   {
    type=MovieScene3DTransformTrack
    sections [{
      type=MovieScene3DTransformSection
      channels [
       Location.X {type=MovieSceneScriptingDoubleChannel, default=0, const [-7153.3203,1]}
       Location.Y {type=MovieSceneScriptingDoubleChannel, default=0, const [-901.6919,1]}
       Location.Z {type=MovieSceneScriptingDoubleChannel, default=0, const [536.5154,1]}
       Rotation.X {type=MovieSceneScriptingDoubleChannel, default=0, const [0,3]}
       Rotation.Y {type=MovieSceneScriptingDoubleChannel, default=0, const [0,3]}
       Rotation.Z {type=MovieSceneScriptingDoubleChannel, default=0, const [0,3]}
       Scale.X {type=MovieSceneScriptingDoubleChannel, default=1, keys [[175,0.001],[183.0013,0.001],[184.0013,0.1],[196.0013,1],[197.0013,0.0001]]}
       Scale.Y {type=MovieSceneScriptingDoubleChannel, default=1, keys [[175,0.001],[183.0013,0.001],[184.0013,0.1],[196.0013,1],[197.0013,0.0001]]}
       Scale.Z {
        type=MovieSceneScriptingDoubleChannel
        default=1
        keys [[175,0.001],[183.0013,0.001],[184.0013,0.1],[188.79,1.5],[196.0013,0.05],[197.0013,0]]
       }
      ]
     }]
   }
   {
    type=MovieSceneSpawnTrack
    sections [{type=MovieSceneSpawnSection, channels [None {type=MovieSceneScriptingBoolChannel, default=true, const [true,1]}]}]
   }
  ]
 }
 NS Sparks Lightning Teleport Poof {
  id=#25
  possessed_class=NiagaraActor
  tracks [
   {type=MovieSceneSpawnTrack, sections [{type=MovieSceneSpawnSection, channels [None {type=MovieSceneScriptingBoolChannel, default=true}]}]}
   {
    type=MovieScene3DTransformTrack
    sections [{
      type=MovieScene3DTransformSection
      channels [
       Location.X {type=MovieSceneScriptingDoubleChannel, default=-6849.1834}
       Location.Y {type=MovieSceneScriptingDoubleChannel, default=-603.2899}
       Location.Z {type=MovieSceneScriptingDoubleChannel, default=892.6262}
       Rotation.X {type=MovieSceneScriptingDoubleChannel, default=0}
       Rotation.Y {type=MovieSceneScriptingDoubleChannel, default=0}
       Rotation.Z {type=MovieSceneScriptingDoubleChannel, default=0}
       Scale.X {type=MovieSceneScriptingDoubleChannel, default=1}
       Scale.Y {type=MovieSceneScriptingDoubleChannel, default=1}
       Scale.Z {type=MovieSceneScriptingDoubleChannel, default=1}
      ]
     }]
   }
  ]
 }
 NiagaraComponent0 {
  id=#26
  parent_id=#5
  possessed_class=NiagaraComponent
  tracks [
   {
    type=MovieSceneParticleTrack
    display_name=FX System
    sections [{type=MovieSceneParticleSection, channels [None {type=MovieSceneScriptingParticleChannel}]}]
   }
   {type=MovieSceneNiagaraSystemTrack, sections [{type=MovieSceneNiagaraSystemSpawnSection, end_frame=480, start_frame=180}]}
  ]
 }
 NS_Rift_Teleport_PlayerGlow {
  id=#22
  possessed_class=NiagaraActor
  tracks [
   {type=MovieSceneSpawnTrack, sections [{type=MovieSceneSpawnSection, channels [None {type=MovieSceneScriptingBoolChannel, default=true}]}]}
   {
    type=MovieScene3DTransformTrack
    sections [{
      type=MovieScene3DTransformSection
      channels [
       Location.X {type=MovieSceneScriptingDoubleChannel, default=-6849.1836}
       Location.Y {type=MovieSceneScriptingDoubleChannel, default=-603.2899}
       Location.Z {type=MovieSceneScriptingDoubleChannel, default=892.6262}
       Rotation.X {type=MovieSceneScriptingDoubleChannel, default=0}
       Rotation.Y {type=MovieSceneScriptingDoubleChannel, default=0}
       Rotation.Z {type=MovieSceneScriptingDoubleChannel, default=0}
       Scale.X {type=MovieSceneScriptingDoubleChannel, default=1}
       Scale.Y {type=MovieSceneScriptingDoubleChannel, default=1}
       Scale.Z {type=MovieSceneScriptingDoubleChannel, default=1}
      ]
     }]
   }
  ]
 }
 NiagaraComponent0 {
  id=#27
  parent_id=#10
  possessed_class=NiagaraComponent
  tracks [
   {
    type=MovieSceneParticleTrack
    display_name=FX System
    sections [{type=MovieSceneParticleSection, channels [None {type=MovieSceneScriptingParticleChannel}]}]
   }
   {type=MovieSceneNiagaraSystemTrack, sections [{type=MovieSceneNiagaraSystemSpawnSection, end_frame=303, start_frame=3}]}
  ]
 }
 NiagaraComponent0 {
  id=#28
  parent_id=#25
  possessed_class=NiagaraComponent
  tracks [{type=MovieSceneNiagaraSystemTrack, sections [{type=MovieSceneNiagaraSystemSpawnSection, end_frame=237, start_frame=166}]}]
 }
 NS_Sparks_Lightning_Teleport_Bolt {
  id=#17
  possessed_class=NiagaraActor
  tracks [
   {type=MovieSceneSpawnTrack, sections [{type=MovieSceneSpawnSection, channels [None {type=MovieSceneScriptingBoolChannel, default=true}]}]}
   {
    type=MovieScene3DTransformTrack
    sections [{
      type=MovieScene3DTransformSection
      channels [
       Location.X {type=MovieSceneScriptingDoubleChannel, default=-7438.1943}
       Location.Y {type=MovieSceneScriptingDoubleChannel, default=-608.8272}
       Location.Z {type=MovieSceneScriptingDoubleChannel, default=785.1509}
       Rotation.X {type=MovieSceneScriptingDoubleChannel, default=-13.7709}
       Rotation.Y {type=MovieSceneScriptingDoubleChannel, default=-77.989}
       Rotation.Z {type=MovieSceneScriptingDoubleChannel, default=15.811}
       Scale.X {type=MovieSceneScriptingDoubleChannel, default=1}
       Scale.Y {type=MovieSceneScriptingDoubleChannel, default=1}
       Scale.Z {type=MovieSceneScriptingDoubleChannel, default=1.25}
      ]
     }]
   }
  ]
 }
 NS Rift Teleport Portal_first {
  id=#29
  possessed_class=NiagaraActor
  tracks [
   {
    type=MovieSceneSpawnTrack
    sections [{type=MovieSceneSpawnSection, channels [None {type=MovieSceneScriptingBoolChannel, default=false, keys [[30,true],[80,false]]}]}]
   }
   {
    type=MovieScene3DTransformTrack
    sections [{
      type=MovieScene3DTransformSection
      channels [
       Location.X {type=MovieSceneScriptingDoubleChannel, default=-6849.1834}
       Location.Y {type=MovieSceneScriptingDoubleChannel, default=-603.2899}
       Location.Z {type=MovieSceneScriptingDoubleChannel, default=892.6262}
       Rotation.X {type=MovieSceneScriptingDoubleChannel, default=0}
       Rotation.Y {type=MovieSceneScriptingDoubleChannel, default=0}
       Rotation.Z {type=MovieSceneScriptingDoubleChannel, default=0}
       Scale.X {type=MovieSceneScriptingDoubleChannel, default=1}
       Scale.Y {type=MovieSceneScriptingDoubleChannel, default=1}
       Scale.Z {type=MovieSceneScriptingDoubleChannel, default=1}
      ]
     }]
   }
  ]
 }
 SM Cylinder Floor Pad Glow {
  id=#30
  possessed_class=FortStaticMeshActor
  tracks [
   {
    type=MovieScene3DTransformTrack
    sections [{
      type=MovieScene3DTransformSection
      channels [
       Location.X {type=MovieSceneScriptingDoubleChannel, default=0, const [-7207.5327,1]}
       Location.Y {type=MovieSceneScriptingDoubleChannel, default=0, const [-340.961,1]}
       Location.Z {type=MovieSceneScriptingDoubleChannel, default=0, const [501.0222,1]}
       Rotation.X {type=MovieSceneScriptingDoubleChannel, default=0, const [0,3]}
       Rotation.Y {type=MovieSceneScriptingDoubleChannel, default=0, const [0,3]}
       Rotation.Z {type=MovieSceneScriptingDoubleChannel, default=0, const [0,3]}
       Scale.X {type=MovieSceneScriptingDoubleChannel, default=1, keys [[175,0.001],[183.0013,0.001],[184.0013,0.1],[196.0013,1],[197.0013,0.0001]]}
       Scale.Y {type=MovieSceneScriptingDoubleChannel, default=1, keys [[175,0.001],[183.0013,0.001],[184.0013,0.1],[196.0013,1],[197.0013,0.0001]]}
       Scale.Z {
        type=MovieSceneScriptingDoubleChannel
        default=1
        keys [[175,0.001],[183.0013,0.001],[184.0013,0.1],[188.79,1.5],[196.0013,0.05],[197.0013,0]]
       }
      ]
     }]
   }
   {
    type=MovieSceneSpawnTrack
    sections [{type=MovieSceneSpawnSection, channels [None {type=MovieSceneScriptingBoolChannel, default=true, const [true,1]}]}]
   }
  ]
 }
 retime_actor {
  id=#31
  possessed_class=Actor
  tracks [
   {type=MovieSceneSpawnTrack, sections [{type=MovieSceneSpawnSection, channels [None {type=MovieSceneScriptingBoolChannel, default=true}]}]}
   {
    type=MovieScene3DTransformTrack
    sections [{
      type=MovieScene3DTransformSection
      channels [
       Location.X {type=MovieSceneScriptingDoubleChannel, default=0, const [0,2]}
       Location.Y {type=MovieSceneScriptingDoubleChannel, default=0, const [0,2]}
       Location.Z {type=MovieSceneScriptingDoubleChannel, default=0, const [0,2]}
       Rotation.X {type=MovieSceneScriptingDoubleChannel, default=0, const [0,2]}
       Rotation.Y {type=MovieSceneScriptingDoubleChannel, default=0, const [0,2]}
       Rotation.Z {type=MovieSceneScriptingDoubleChannel, default=0, const [0,2]}
       Scale.X {type=MovieSceneScriptingDoubleChannel, default=1, const [1,2]}
       Scale.Y {type=MovieSceneScriptingDoubleChannel, default=1, const [1,2]}
       Scale.Z {type=MovieSceneScriptingDoubleChannel, default=1, const [1,2]}
      ]
     }]
   }
  ]
 }
 NS_Sparks_Lightning_Teleport_Bolt {
  id=#9
  possessed_class=NiagaraActor
  tracks [
   {type=MovieSceneSpawnTrack, sections [{type=MovieSceneSpawnSection, channels [None {type=MovieSceneScriptingBoolChannel, default=true}]}]}
   {
    type=MovieScene3DTransformTrack
    sections [{
      type=MovieScene3DTransformSection
      channels [
       Location.X {type=MovieSceneScriptingDoubleChannel, default=-7109.7249}
       Location.Y {type=MovieSceneScriptingDoubleChannel, default=-764.7314}
       Location.Z {type=MovieSceneScriptingDoubleChannel, default=677.4735}
       Rotation.X {type=MovieSceneScriptingDoubleChannel, default=56.4285}
       Rotation.Y {type=MovieSceneScriptingDoubleChannel, default=0.1659}
       Rotation.Z {type=MovieSceneScriptingDoubleChannel, default=-48.269}
       Scale.X {type=MovieSceneScriptingDoubleChannel, default=1}
       Scale.Y {type=MovieSceneScriptingDoubleChannel, default=1}
       Scale.Z {type=MovieSceneScriptingDoubleChannel, default=0.8}
      ]
     }]
   }
  ]
 }
 NiagaraComponent0 {
  id=#32
  parent_id=#19
  possessed_class=NiagaraComponent
  tracks [
   {
    type=MovieSceneParticleTrack
    display_name=FX System
    sections [{type=MovieSceneParticleSection, channels [None {type=MovieSceneScriptingParticleChannel}]}]
   }
   {type=MovieSceneNiagaraSystemTrack, sections [{type=MovieSceneNiagaraSystemSpawnSection, end_frame=466, start_frame=166}]}
   {
    type=MovieSceneNiagaraVectorParameterTrack
    display_name=EndPosition
    sections [{
      type=MovieSceneFloatVectorSection
      channels [
       Vector.X {type=MovieSceneScriptingFloatChannel, default=-575}
       Vector.Y {type=MovieSceneScriptingFloatChannel, default=0}
       Vector.Z {type=MovieSceneScriptingFloatChannel, default=-80}
      ]
     }]
   }
  ]
 }
 SM Hex Floor Pad Glow {
  id=#33
  possessed_class=FortStaticMeshActor
  tracks [
   {
    type=MovieScene3DTransformTrack
    sections [{
      type=MovieScene3DTransformSection
      channels [
       Location.X {type=MovieSceneScriptingDoubleChannel, default=0, const [-6904.856,1]}
       Location.Y {type=MovieSceneScriptingDoubleChannel, default=0, const [-587.4029,1]}
       Location.Z {type=MovieSceneScriptingDoubleChannel, default=0, const [497.0325,1]}
       Rotation.X {type=MovieSceneScriptingDoubleChannel, default=0, const [0,3]}
       Rotation.Y {type=MovieSceneScriptingDoubleChannel, default=0, const [0,3]}
       Rotation.Z {type=MovieSceneScriptingDoubleChannel, default=0, const [0,3]}
       Scale.X {type=MovieSceneScriptingDoubleChannel, default=1, keys [[175,0.001],[183.0013,0.001],[184.0013,0.1],[196.0013,1],[197.0013,0.0001]]}
       Scale.Y {type=MovieSceneScriptingDoubleChannel, default=1, keys [[175,0.001],[183.0013,0.001],[184.0013,0.1],[196.0013,1],[197.0013,0.0001]]}
       Scale.Z {
        type=MovieSceneScriptingDoubleChannel
        default=1
        keys [[175,0.001],[183.0013,0.001],[184.0013,0.1],[188.79,1.5],[196.0013,0.05],[197.0013,0]]
       }
      ]
     }]
   }
   {
    type=MovieSceneSpawnTrack
    sections [{type=MovieSceneSpawnSection, channels [None {type=MovieSceneScriptingBoolChannel, default=true, const [true,1]}]}]
   }
  ]
 }
 NiagaraComponent0 {
  id=#34
  parent_id=#35
  possessed_class=NiagaraComponent
  tracks [
   {
    type=MovieSceneParticleTrack
    display_name=FX System
    sections [{type=MovieSceneParticleSection, channels [None {type=MovieSceneScriptingParticleChannel}]}]
   }
   {type=MovieSceneNiagaraSystemTrack, sections [{type=MovieSceneNiagaraSystemSpawnSection, end_frame=479, start_frame=179}]}
  ]
 }
 NiagaraComponent0 {
  id=#36
  parent_id=#29
  possessed_class=NiagaraComponent
  tracks [
   {
    type=MovieSceneParticleTrack
    display_name=FX System
    sections [{type=MovieSceneParticleSection, channels [None {type=MovieSceneScriptingParticleChannel}]}]
   }
   {type=MovieSceneNiagaraSystemTrack, sections [{type=MovieSceneNiagaraSystemSpawnSection, end_frame=331, start_frame=31}]}
  ]
 }
 NiagaraComponent0 {
  id=#37
  parent_id=#15
  possessed_class=NiagaraComponent
  tracks [
   {
    type=MovieSceneParticleTrack
    display_name=FX System
    sections [{type=MovieSceneParticleSection, channels [None {type=MovieSceneScriptingParticleChannel}]}]
   }
   {type=MovieSceneNiagaraSystemTrack, sections [{type=MovieSceneNiagaraSystemSpawnSection, end_frame=469, start_frame=169}]}
  ]
 }
 NS_Sparks_Lightning_Teleport_Blast {
  id=#35
  possessed_class=NiagaraActor
  tracks [
   {type=MovieSceneSpawnTrack, sections [{type=MovieSceneSpawnSection, channels [None {type=MovieSceneScriptingBoolChannel, default=true}]}]}
   {
    type=MovieScene3DTransformTrack
    sections [{
      type=MovieScene3DTransformSection
      channels [
       Location.X {type=MovieSceneScriptingDoubleChannel, default=-7109.3818}
       Location.Y {type=MovieSceneScriptingDoubleChannel, default=-765.5574}
       Location.Z {type=MovieSceneScriptingDoubleChannel, default=680.5154}
       Rotation.X {type=MovieSceneScriptingDoubleChannel, default=0}
       Rotation.Y {type=MovieSceneScriptingDoubleChannel, default=0}
       Rotation.Z {type=MovieSceneScriptingDoubleChannel, default=-79.1304}
       Scale.X {type=MovieSceneScriptingDoubleChannel, default=1}
       Scale.Y {type=MovieSceneScriptingDoubleChannel, default=1}
       Scale.Z {type=MovieSceneScriptingDoubleChannel, default=1}
      ]
     }]
   }
  ]
 }
 NS_Sparks_Lightning_Teleport_Blast {
  id=#14
  possessed_class=NiagaraActor
  tracks [
   {type=MovieSceneSpawnTrack, sections [{type=MovieSceneSpawnSection, channels [None {type=MovieSceneScriptingBoolChannel, default=true}]}]}
   {
    type=MovieScene3DTransformTrack
    sections [{
      type=MovieScene3DTransformSection
      channels [
       Location.X {type=MovieSceneScriptingDoubleChannel, default=-6911.3247}
       Location.Y {type=MovieSceneScriptingDoubleChannel, default=-462.9917}
       Location.Z {type=MovieSceneScriptingDoubleChannel, default=629.0325}
       Rotation.X {type=MovieSceneScriptingDoubleChannel, default=0}
       Rotation.Y {type=MovieSceneScriptingDoubleChannel, default=0}
       Rotation.Z {type=MovieSceneScriptingDoubleChannel, default=-39.421}
       Scale.X {type=MovieSceneScriptingDoubleChannel, default=1}
       Scale.Y {type=MovieSceneScriptingDoubleChannel, default=1}
       Scale.Z {type=MovieSceneScriptingDoubleChannel, default=1}
      ]
     }]
   }
  ]
 }
 NiagaraComponent0 {
  id=#38
  parent_id=#18
  possessed_class=NiagaraComponent
  tracks [
   {
    type=MovieSceneParticleTrack
    display_name=FX System
    sections [{type=MovieSceneParticleSection, channels [None {type=MovieSceneScriptingParticleChannel}]}]
   }
   {type=MovieSceneNiagaraSystemTrack, sections [{type=MovieSceneNiagaraSystemSpawnSection, end_frame=480, start_frame=180}]}
  ]
 }
 SM Cylinder Floor Pad Glow {
  id=#39
  possessed_class=FortStaticMeshActor
  tracks [
   {
    type=MovieScene3DTransformTrack
    sections [{
      type=MovieScene3DTransformSection
      channels [
       Location.X {type=MovieSceneScriptingDoubleChannel, default=0, const [-7447.4619,1]}
       Location.Y {type=MovieSceneScriptingDoubleChannel, default=0, const [-716.4758,1]}
       Location.Z {type=MovieSceneScriptingDoubleChannel, default=0, const [653.6868,1]}
       Rotation.X {type=MovieSceneScriptingDoubleChannel, default=0, const [0,3]}
       Rotation.Y {type=MovieSceneScriptingDoubleChannel, default=0, const [0,3]}
       Rotation.Z {type=MovieSceneScriptingDoubleChannel, default=0, const [0,3]}
       Scale.X {type=MovieSceneScriptingDoubleChannel, default=1, keys [[175,0.001],[183.0013,0.001],[184.0013,0.1],[196.0013,1],[197.0013,0.0001]]}
       Scale.Y {type=MovieSceneScriptingDoubleChannel, default=1, keys [[175,0.001],[183.0013,0.001],[184.0013,0.1],[196.0013,1],[197.0013,0.0001]]}
       Scale.Z {
        type=MovieSceneScriptingDoubleChannel
        default=1
        keys [[175,0.001],[183.0013,0.001],[184.0013,0.1],[188.79,1.5],[196.0013,0.05],[197.0013,0]]
       }
      ]
     }]
   }
   {
    type=MovieSceneSpawnTrack
    sections [{type=MovieSceneSpawnSection, channels [None {type=MovieSceneScriptingBoolChannel, default=true, const [true,1]}]}]
   }
  ]
 }
]
display_rate {denominator=1, numerator=30}
tick_resolution {denominator=1, numerator=24000}
tracks [
 {type=MovieSceneCinematicShotTrack, display_name=Shots, sections [{type=MovieSceneCinematicShotSection, end_frame=496, start_frame=0}]}
 {type=MovieSceneSubTrack, display_name=Subsequences, sections [{type=MovieSceneSubSection, end_frame=240, start_frame=0}]}
]
