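# Level Sequence animatic for the Pilgrim Quickplay band intro teleport, staging four cine teleporter pawns and four player mannequins.
# Pawns hold elevated poses, with Pawn0 drifting slightly, and fire events at start, middle and end.
# Each mannequin's keyed move starts at or near a pawn position and descends to a separate lower spot, Player1 leading the others.
# A static mesh EventObject carries Show/Hide and secondary event tracks timed around those moves.
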LevelSequence /PilgrimQuickplay/BandIntro/Teleport/BandIntro_Animatic_MASTER/LS_BI_AM_playerPrim_4Pawns
bindings [
 Cine_Player4 {
  id=#1
  possessed_class=FortPlayerMannequin
  tracks [
   {
    type=MovieScene3DTransformTrack
    sections [{
      type=MovieScene3DTransformSection
      channels [
       Location.X {type=MovieSceneScriptingDoubleChannel, default=-7000, keys [[64,-6849],[188,-7447.4619]]}
       Location.Y {type=MovieSceneScriptingDoubleChannel, default=-600, keys [[64,-573],[188,-716.4758]]}
       Location.Z {type=MovieSceneScriptingDoubleChannel, default=570, keys [[64,870],[188,718.2806]]}
       Rotation.X {type=MovieSceneScriptingDoubleChannel, default=0, const [0,2]}
       Rotation.Y {type=MovieSceneScriptingDoubleChannel, default=0, const [0,2]}
       Rotation.Z {type=MovieSceneScriptingDoubleChannel, default=0, const [0,2]}
       Scale.X {type=MovieSceneScriptingDoubleChannel, default=1, const [1,2]}
       Scale.Y {type=MovieSceneScriptingDoubleChannel, default=1, const [1,2]}
       Scale.Z {type=MovieSceneScriptingDoubleChannel, default=1, const [1,2]}
      ]
     }]
   }
   {type=MovieSceneSpawnTrack, sections [{type=MovieSceneSpawnSection, channels [None {type=MovieSceneScriptingBoolChannel, default=true}]}]}
  ]
 }
 Cine_Pawn2 {
  id=#2
  possessed_class=Character_Cine_Teleporter_C
  tracks [
   {type=MovieSceneSpawnTrack, sections [{type=MovieSceneSpawnSection, channels [None {type=MovieSceneScriptingBoolChannel, default=true}]}]}
   {
    type=MovieScene3DTransformTrack
    sections [{
      type=MovieScene3DTransformSection
      channels [
       Location.X {type=MovieSceneScriptingDoubleChannel, default=-6905}
       Location.Y {type=MovieSceneScriptingDoubleChannel, default=-682}
       Location.Z {type=MovieSceneScriptingDoubleChannel, default=918}
       Rotation.X {type=MovieSceneScriptingDoubleChannel, default=22.6974}
       Rotation.Y {type=MovieSceneScriptingDoubleChannel, default=-11.716}
       Rotation.Z {type=MovieSceneScriptingDoubleChannel, default=-25.8481}
       Scale.X {type=MovieSceneScriptingDoubleChannel, default=1}
       Scale.Y {type=MovieSceneScriptingDoubleChannel, default=1}
       Scale.Z {type=MovieSceneScriptingDoubleChannel, default=1}
      ]
     }]
   }
   {
    type=MovieSceneEventTrack
    sections [{
      type=MovieSceneEventTriggerSection
      channels [None {
        type=MovieSceneScriptingEventChannel
        keys [
         [0, "<Struct 'MovieSceneEvent' (0x8523896e420) {}>"]
         [60, "<Struct 'MovieSceneEvent' (0x8523896dfc0) {}>"]
         [187, "<Struct 'MovieSceneEvent' (0x8523896d5c0) {}>"]
        ]
       }]
     }]
   }
  ]
 }
 Cine_Pawn3 {
  id=#3
  possessed_class=Character_Cine_Teleporter_C
  tracks [
   {type=MovieSceneSpawnTrack, sections [{type=MovieSceneSpawnSection, channels [None {type=MovieSceneScriptingBoolChannel, default=true}]}]}
   {
    type=MovieScene3DTransformTrack
    sections [{
      type=MovieScene3DTransformSection
      channels [
       Location.X {type=MovieSceneScriptingDoubleChannel, default=-6824.67}
       Location.Y {type=MovieSceneScriptingDoubleChannel, default=-573}
       Location.Z {type=MovieSceneScriptingDoubleChannel, default=893.7717}
       Rotation.X {type=MovieSceneScriptingDoubleChannel, default=0}
       Rotation.Y {type=MovieSceneScriptingDoubleChannel, default=-20}
       Rotation.Z {type=MovieSceneScriptingDoubleChannel, default=0}
       Scale.X {type=MovieSceneScriptingDoubleChannel, default=1}
       Scale.Y {type=MovieSceneScriptingDoubleChannel, default=1}
       Scale.Z {type=MovieSceneScriptingDoubleChannel, default=1}
      ]
     }]
   }
   {
    type=MovieSceneEventTrack
    sections [{
      type=MovieSceneEventTriggerSection
      channels [None {
        type=MovieSceneScriptingEventChannel
        keys [
         [0, "<Struct 'MovieSceneEvent' (0x8523896f500) {}>"]
         [60, "<Struct 'MovieSceneEvent' (0x8523896f5a0) {}>"]
         [187, "<Struct 'MovieSceneEvent' (0x8523896f640) {}>"]
        ]
       }]
     }]
   }
  ]
 }
 EventObject {
  id=#4
  possessed_class=FortStaticMeshActor
  tracks [
   {type=MovieSceneSpawnTrack, sections [{type=MovieSceneSpawnSection, channels [None {type=MovieSceneScriptingBoolChannel, default=true}]}]}
   {
    type=MovieSceneEventTrack
    display_name=Show/Hide Events
    sections [{
      type=MovieSceneEventTriggerSection
      channels [None {
        type=MovieSceneScriptingEventChannel
        keys [
         [0, "<Struct 'MovieSceneEvent' (0x8523896f000) {}>"]
         [64, "<Struct 'MovieSceneEvent' (0x8523896ef60) {}>"]
         [163, "<Struct 'MovieSceneEvent' (0x8523896eec0) {}>"]
         [188, "<Struct 'MovieSceneEvent' (0x8523896ee20) {}>"]
        ]
       }]
     }]
   }
   {
    type=MovieSceneEventTrack
    sections [{
      type=MovieSceneEventTriggerSection
      channels [None {
        type=MovieSceneScriptingEventChannel
        keys [[63, "<Struct 'MovieSceneEvent' (0x8523896b680) {}>"], [187, "<Struct 'MovieSceneEvent' (0x8523896e7e0) {}>"]]
       }]
     }]
   }
  ]
 }
 Cine_Pawn0 {
  id=#5
  possessed_class=Character_Cine_Teleporter_C
  tracks [
   {type=MovieSceneSpawnTrack, sections [{type=MovieSceneSpawnSection, channels [None {type=MovieSceneScriptingBoolChannel, default=true}]}]}
   {
    type=MovieScene3DTransformTrack
    sections [{
      type=MovieScene3DTransformSection
      channels [
       Location.X {type=MovieSceneScriptingDoubleChannel, default=-6882, keys [[0,-6912.5103],[166,-6882]]}
       Location.Y {type=MovieSceneScriptingDoubleChannel, default=-522, keys [[0,-525.7588],[166,-522]]}
       Location.Z {type=MovieSceneScriptingDoubleChannel, default=905, keys [[0,923.7361],[166,905]]}
       Rotation.X {type=MovieSceneScriptingDoubleChannel, default=0}
       Rotation.Y {type=MovieSceneScriptingDoubleChannel, default=-30}
       Rotation.Z {type=MovieSceneScriptingDoubleChannel, default=30}
       Scale.X {type=MovieSceneScriptingDoubleChannel, default=1}
       Scale.Y {type=MovieSceneScriptingDoubleChannel, default=1}
       Scale.Z {type=MovieSceneScriptingDoubleChannel, default=1}
      ]
     }]
   }
   {
    type=MovieSceneEventTrack
    sections [{
      type=MovieSceneEventTriggerSection
      channels [None {
        type=MovieSceneScriptingEventChannel
        keys [
         [0, "<Struct 'MovieSceneEvent' (0x8523896f1e0) {}>"]
         [60, "<Struct 'MovieSceneEvent' (0x8523896f140) {}>"]
         [187, "<Struct 'MovieSceneEvent' (0x8523896f0a0) {}>"]
        ]
       }]
     }]
   }
  ]
 }
 Cine_Player1 {
  id=#6
  possessed_class=FortPlayerMannequin
  tracks [
   {
    type=MovieScene3DTransformTrack
    sections [{
      type=MovieScene3DTransformSection
      channels [
       Location.X {type=MovieSceneScriptingDoubleChannel, default=-7000, keys [[49,-6882.5449],[173,-7207.5327]]}
       Location.Y {type=MovieSceneScriptingDoubleChannel, default=-600, keys [[49,-522],[173,-340]]}
       Location.Z {type=MovieSceneScriptingDoubleChannel, default=570, keys [[49,915],[173,570]]}
       Rotation.X {type=MovieSceneScriptingDoubleChannel, default=0, const [0,2]}
       Rotation.Y {type=MovieSceneScriptingDoubleChannel, default=0, const [0,2]}
       Rotation.Z {type=MovieSceneScriptingDoubleChannel, default=0, const [0,2]}
       Scale.X {type=MovieSceneScriptingDoubleChannel, default=1, const [1,2]}
       Scale.Y {type=MovieSceneScriptingDoubleChannel, default=1, const [1,2]}
       Scale.Z {type=MovieSceneScriptingDoubleChannel, default=1, const [1,2]}
      ]
     }]
   }
   {type=MovieSceneSpawnTrack, sections [{type=MovieSceneSpawnSection, channels [None {type=MovieSceneScriptingBoolChannel, default=true}]}]}
  ]
 }
 Cine_Player3 {
  id=#7
  possessed_class=FortPlayerMannequin
  tracks [
   {
    type=MovieScene3DTransformTrack
    sections [{
      type=MovieScene3DTransformSection
      channels [
       Location.X {type=MovieSceneScriptingDoubleChannel, default=-7000, keys [[64,-6905],[188,-7153]]}
       Location.Y {type=MovieSceneScriptingDoubleChannel, default=-600, keys [[64,-682],[188,-901]]}
       Location.Z {type=MovieSceneScriptingDoubleChannel, default=570, keys [[64,918],[188,628]]}
       Rotation.X {type=MovieSceneScriptingDoubleChannel, default=0, const [0,2]}
       Rotation.Y {type=MovieSceneScriptingDoubleChannel, default=0, const [0,2]}
       Rotation.Z {type=MovieSceneScriptingDoubleChannel, default=0, const [0,2]}
       Scale.X {type=MovieSceneScriptingDoubleChannel, default=1, const [1,2]}
       Scale.Y {type=MovieSceneScriptingDoubleChannel, default=1, const [1,2]}
       Scale.Z {type=MovieSceneScriptingDoubleChannel, default=1, const [1,2]}
      ]
     }]
   }
   {type=MovieSceneSpawnTrack, sections [{type=MovieSceneSpawnSection, channels [None {type=MovieSceneScriptingBoolChannel, default=true}]}]}
  ]
 }
 Cine_Pawn1 {
  id=#8
  possessed_class=Character_Cine_Teleporter_C
  tracks [
   {type=MovieSceneSpawnTrack, sections [{type=MovieSceneSpawnSection, channels [None {type=MovieSceneScriptingBoolChannel, default=true}]}]}
   {
    type=MovieScene3DTransformTrack
    sections [{
      type=MovieScene3DTransformSection
      channels [
       Location.X {type=MovieSceneScriptingDoubleChannel, default=-6924}
       Location.Y {type=MovieSceneScriptingDoubleChannel, default=-611}
       Location.Z {type=MovieSceneScriptingDoubleChannel, default=940}
       Rotation.X {type=MovieSceneScriptingDoubleChannel, default=3.6164}
       Rotation.Y {type=MovieSceneScriptingDoubleChannel, default=-19.6835}
       Rotation.Z {type=MovieSceneScriptingDoubleChannel, default=-10.6276}
       Scale.X {type=MovieSceneScriptingDoubleChannel, default=1}
       Scale.Y {type=MovieSceneScriptingDoubleChannel, default=1}
       Scale.Z {type=MovieSceneScriptingDoubleChannel, default=1}
      ]
     }]
   }
   {
    type=MovieSceneEventTrack
    sections [{
      type=MovieSceneEventTriggerSection
      channels [None {
        type=MovieSceneScriptingEventChannel
        keys [
         [0, "<Struct 'MovieSceneEvent' (0x8523896ea60) {}>"]
         [60, "<Struct 'MovieSceneEvent' (0x8523896eb00) {}>"]
         [187, "<Struct 'MovieSceneEvent' (0x8523896eba0) {}>"]
        ]
       }]
     }]
   }
  ]
 }
 Cine_Player2 {
  id=#9
  possessed_class=FortPlayerMannequin
  tracks [
   {
    type=MovieScene3DTransformTrack
    sections [{
      type=MovieScene3DTransformSection
      channels [
       Location.X {type=MovieSceneScriptingDoubleChannel, default=0, keys [[64,-6924],[188,-6904]]}
       Location.Y {type=MovieSceneScriptingDoubleChannel, default=0, keys [[64,-611],[188,-587]]}
       Location.Z {type=MovieSceneScriptingDoubleChannel, default=0, keys [[64,925],[188,548]]}
       Rotation.X {type=MovieSceneScriptingDoubleChannel, default=0, const [0,2]}
       Rotation.Y {type=MovieSceneScriptingDoubleChannel, default=0, const [0,2]}
       Rotation.Z {type=MovieSceneScriptingDoubleChannel, default=0, const [0,2]}
       Scale.X {type=MovieSceneScriptingDoubleChannel, default=1, const [1,2]}
       Scale.Y {type=MovieSceneScriptingDoubleChannel, default=1, const [1,2]}
       Scale.Z {type=MovieSceneScriptingDoubleChannel, default=1, const [1,2]}
      ]
     }]
   }
   {type=MovieSceneSkeletalAnimationTrack, sections []}
   {type=MovieSceneSpawnTrack, sections [{type=MovieSceneSpawnSection, channels [None {type=MovieSceneScriptingBoolChannel, default=true}]}]}
  ]
 }
]
display_rate {denominator=1, numerator=30}
tick_resolution {denominator=1, numerator=24000}
tracks []
